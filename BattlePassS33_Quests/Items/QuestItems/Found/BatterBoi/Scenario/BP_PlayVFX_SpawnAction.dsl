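; BP_PlayVFX_SpawnAction is a Living World spawn action in the BatterBoi quest scenario that attaches a VFX component to spawned pawns.
; When all actors have spawned, it loops over them, skips invalid ones and casts each to a Fort pawn.
; It then adds BP_ReplicatedPawnVFXComponent to each pawn, leaving the effect itself to that component.
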(fn OnAllActorSpawned
  (params (SpawnedActor :type Array%20of%20Actor%20Object%20References) (SourcePointProvider :type Fort%20Athena%20Living%20World%20Point%20Provider%20Interface%20Interface%20%28by%20ref%29))
  (for _array_element SpawnedActor
    (Utilities|IsValid _array_element
      (:Is%20Valid
        (bind _asfort_pawn (Utilities|Casting|CastToFortPawn _array_element))
        (Game|AddComponentbyClass _asfort_pawn "/BattlePassS33_Quests/Items/QuestItems/Found/BatterBoi/Scenario/BP_ReplicatedPawnVFXComponent.BP_ReplicatedPawnVFXComponent_C"))))
  (return))
