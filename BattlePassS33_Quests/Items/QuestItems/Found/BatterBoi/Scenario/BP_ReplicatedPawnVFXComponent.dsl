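; Actor component in the Season 33 BatterBoi quest scenario that shows a Niagara effect on its owning pawn.
; On BeginPlay it runs only without authority, so the effect is created locally on clients rather than the server.
; It reuses the owner's Niagara component by assigning the VFX variable and activating it.
; If none exists, it casts to FortPawn and spawns the system attached to the skeletal mesh returned by a FortVehicleInterface call.
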(event EventBeginPlay
  (bind _returnvalue (Components|GetOwner))
  (bind _returnvalue_2 (Actor|GetComponentbyClass _returnvalue "/Script/Niagara.NiagaraComponent"))
  (bind _vfx (Variables|Default|GetVFX))
  (Utilities|SwitchHasAuthority
    (:Remote
      (Utilities|IsValid _returnvalue
        (:Is%20Valid
          (Utilities|IsValid _returnvalue_2
            (:Is%20Valid
              (Niagara|SetNiagaraSystemAsset _returnvalue_2 _vfx true)
              (Components|Activation|Activate _returnvalue_2 true))
            (:Is%20Not%20Valid
              (bind _asfort_pawn (Utilities|Casting|CastToFortPawn _returnvalue))
              (bind _returnvalue_3 (Class|FortVehicleInterface|GetSkeletalMeshComponent(Message) _asfort_pawn))
              (Niagara|SpawnSystemAttached _vfx _returnvalue_3))))))))

(event EventTick
  (params (DeltaSeconds :type Float%20%28single-precision%29)))
