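# StateTree for the BatterBoi quest scenario, staging three henchman assault waves on the BatterBoi NPC, ending in a chest reward.
# Once players are present, it grants BatterBoi protective reactions, then each stage shows a speech bubble, spawns enemies and targets him.
# Remaining-count tasks advance cleared waves, with a lone boss as wave three.
# Boss defeat jumps to Speech_Bubble, bypassing Go_To_Chest, then Spawn_Chest grants the second playspace quest and spawns the chest.
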StateTree /BattlePassS33_Quests/Items/QuestItems/Found/BatterBoi/Scenario/ST_QuestScenario_BatterBoi
global_tasks [{
  id=#1
  description=Get Scenario Actor - BatterBoi
  properties {
   iD=#2
   instance {
    _structType=&GetEncounterActorStateTreeTaskInstanceData
    actorQuery {autoDescription=" ANY( AISpawnerData.Type.Tandem.BatterBoi )", tagDictionary [{tagName=AISpawnerData.Type.Tandem.BatterBoi}]}
   }
   node {name=Get Scenario Actor - BatterBoi, _structType=&GetEncounterActorStateTreeTask}
  }
 }]
states [
 Root {
  id=#3
  type=STATE
  children [
   #4
   #5
   #6
   #7
   #8
   #9
   #10
   #11
   #12
   #13
   #14
   #15
   #16
  ]
  tasks []
  transitions []
 }
 Init {
  id=#4
  type=STATE
  tasks [{
    id=#17
    description="Scenario [Player Count Transition]"
    properties {
     iD=#18
     instance {
      _structType=/Script/EncountersRuntime.EncounterPlayerCountTransitionStateTreeTaskInstanceData
      bOnlyTriggerOnce=true
      bShouldTransition=true
      bTriggered=false
      comparisonOperator=GreaterThanOrEqualTo
      userActor=null
      value=1
      customVerb {amount=1, customVerbTag {tagName=None}, subjects [], values []}
      onTriggered {iD=#19}
     }
     node {
      name=None
      _structType=/Script/EncountersRuntime.EncounterPlayerCountTransitionStateTreeTask
      bCanEditConsideredForCompletion=true
      bConsideredForCompletion=true
      bHasTaskCompletionDelegateDispatcher=false
      bTaskEnabled=true
      transitionHandlingPriority=Normal
      bindingsBatch {value=65535}
      executionRuntimeTemplateIndex {value=65535}
      instanceDataHandle {index=65535, source=None, stateHandle {index=65535}}
      instanceTemplateIndex {value=65535}
      outputBindingsBatch {value=65535}
      transitionTo {name=WaitToStart, fallback=None, iD=#20, linkType=GotoState, stateHandle {index=65535}}
     }
    }
   }]
  transitions []
 }
 WaitToStart {
  id=#5
  type=STATE
  tasks [{
    id=#21
    description=Delay 5.0
    properties {
     iD=#22
     instance {_structType=&StateTreeDelayTaskInstanceData, duration=5}
     node {name=None, _structType=&StateTreeDelayTask}
    }
   }]
  transitions [{id=#23, trigger=ON_STATE_COMPLETED, state {id=#6, name=Grant_Reactions}}]
 }
 Grant_Reactions {
  id=#6
  type=STATE
  tasks [
   {
    id=#24
    description=Grant Reactions - BatterBoi
    properties {
     iD=#25
     instance {
      _structType=/Script/EncountersRuntime.GrantReactionStateTreeTaskInstanceData
      grantType=GrantToTargetActor
      targetActor=null
      userActor=null
      reactionEntries [
       {rowValue {
         bUseDataRegistryId=false
         dataRegistryId=""
         dataTableRow {
          dataTable=&/DT_QuestScenario_BatterBoi_Reactions
          rowName=AlertStateHideIndicator
         }
        }}
       {rowValue {
         bUseDataRegistryId=false
         dataRegistryId=""
         dataTableRow {dataTable=&DT_QuestScenario_BatterBoi_Reactions, rowName=DamageImmunity}
        }}
       {rowValue {
         bUseDataRegistryId=false
         dataRegistryId=""
         dataTableRow {
          dataTable=&DT_QuestScenario_BatterBoi_Reactions
          rowName=OutgoingDamageReduction
         }
        }}
       {rowValue {
         bUseDataRegistryId=false
         dataRegistryId=""
         dataTableRow {
          dataTable=&DT_QuestScenario_BatterBoi_Reactions
          rowName=IconVisibility_Visible
         }
        }}
      ]
      reactions []
     }
     node {
      name=Grant Reactions - BatterBoi
      _structType=/Script/EncountersRuntime.GrantReactionStateTreeTask
      bCanEditConsideredForCompletion=true
      bConsideredForCompletion=true
      bHasTaskCompletionDelegateDispatcher=false
      bTaskEnabled=true
      transitionHandlingPriority=Normal
      bindingsBatch {value=65535}
      executionRuntimeTemplateIndex {value=65535}
      instanceDataHandle {index=65535, source=None, stateHandle {index=65535}}
      instanceTemplateIndex {value=65535}
      outputBindingsBatch {value=65535}
     }
    }
   }
   {
    id=#26
    description=Delay 1.0
    properties {
     iD=#27
     instance {_structType=&StateTreeDelayTaskInstanceData, duration=1}
     node {name=None, _structType=&StateTreeDelayTask}
    }
   }
  ]
  transitions [{id=#28, trigger=ON_STATE_COMPLETED, state {id=#7, name=Stage_Wave_One}}]
 }
 Stage_Wave_One {
  id=#7
  type=STATE
  tasks [
   {
    id=#29
    description=Speech Bubble - BatterBoi
    properties {
     iD=#30
     instance {
      _structType=&GrantReactionStateTreeTaskInstanceData
      grantType=GrantToTargetActor
      targetActor=null
      userActor=null
      reactionEntries [{rowValue {
         bUseDataRegistryId=false
         dataRegistryId=""
         dataTableRow {dataTable=&DT_QuestScenario_BatterBoi_Reactions, rowName=SpeechBubble_WaveOne}
        }}]
      reactions []
     }
     node {
      name=Speech Bubble - BatterBoi
      _structType=&GrantReactionStateTreeTask
      bCanEditConsideredForCompletion=true
      bConsideredForCompletion=true
      bHasTaskCompletionDelegateDispatcher=false
      bTaskEnabled=true
      transitionHandlingPriority=Normal
      bindingsBatch {value=65535}
      executionRuntimeTemplateIndex {value=65535}
      instanceDataHandle {index=65535, source=None, stateHandle {index=65535}}
      instanceTemplateIndex {value=65535}
      outputBindingsBatch {value=65535}
     }
    }
   }
   {
    id=#31
    description=Spawn Wave One
    properties {
     iD=#32
     instance {
      _structType=/Script/EncountersRuntime.LWMEncounterDynamicStartStateTreeTaskInstanceData
      bAttachVolumeToSpawn=false
      bOverrideAnchorPosition=true
      bOverrideDelayBetweenEvents=false
      bOverrideEnvQuery=false
      bOverrideStageActorMaxCount=false
      bOverrideStageActorMaxSpawnedCount=false
      bPauseOnExit=false
      bSpecifyFaction=true
      bUseVolumeAsAnchor=true
      bWaitForEncounterDoneSpawning=false
      delayBetweenEventsMax=0
      delayBetweenEventsMin=0
      dynamicEmptyEncounter=/FortQuestEncounters/Lager/Lager_Encounter_Empty_QuestEncounter
      faction=&/DA_QuestScenario_EnemyGroup
      instigator=null
      owningActor=null
      spawnAroundEnvironmentQuery=null
      spawnAroundEnvironmentQueryRunMode=AllMatching
      spawnAroundNavigationSourceOverride=""
      stageActorMaxCount=1
      stageActorMaxSpawnedCount=-1
      userActor=null
      actorTags {gameplayTags []}
      anchorPointTagQuery {autoDescription="", tokenStreamVersion=0, userDescription="", queryTokenStream [], tagDictionary []}
      dynamicEncounterEntries [{
        actorCountOverrideMax=1
        actorCountOverrideMin=1
        bCreateRandomSingleActorEvents=false
        bOverrideActorCount=false
        bUseEncounterTagQuery=true
        bUseEventTemplate=false
        debugName=""
        maxConcurrentSpawns=2
        noActorLimitEventCount=1
        spawnCount=2
        actorDescriptions []
        encounterTagQuery {
         autoDescription=" ANY( Athena.Quest.Scenario.Faction.Henchmen.Assault )"
         tokenStreamVersion=0
         userDescription=""
         queryTokenStream [0,1,1,1,0]
         tagDictionary [{tagName=Athena.Quest.Scenario.Faction.Henchmen.Assault}]
        }
        eventTemplate {
         activationGamePhaseStep=None
         activationMaxDelay=0
         activationMinDelay=0
         activationSafeZoneIndex=0
         bCanSpawnWithoutMatchingPlayerDistance=false
         bCheckForActorDensity=true
         bOnlyInNextSafeZone=false
         bOnlyInNextSafeZoneIfStormShrinking=false
         bOnlyInSafeZone=true
         bShouldProcessEventRequestImmediately=false
         deactivationMaxDelay=-1
         deactivationMinDelay=-1
         maxActorDensityOverride=-1
         maxDistanceToPlayer=-1
         minDistanceToPlayer=-1
         randomActivationChance=100
         weight=1
         calendarEventConditions {isActiveWithoutSeasonalManager=false, conditions []}
         canSpawnCondition {description=Empty, schemaClass=/Script/LagerRuntime.FortAthenaLivingWorldConditionSchema, editableConditions [], sharedDefinition {}}
         despawnCondition {
          bCanBeDespawned=false
          bCanDespawnInTheStorm=false
          bCanDespawnWhenInCombatWithPlayer=false
          canStartDespawningPhaseStep=None
          canStartDespawningSafeZonePhase=-1
          maxStimulusAgeToBeConsideredInCombat=30
          minDistanceToPlayer=-1
          timeAsCandidateBeforeDespawn=10
         }
         despawnConditions []
         eventDatas []
         eventInstanceMaxCount [-1]
         eventInstanceMaxSpawnedCount [-1]
         instantDeathMaxRespawnTime []
         instantDeathMinRespawnTime []
         isEnabled {registryType=None, value=1, curve {curveTable=null, rowName=None}}
         maxCount [0]
         maxRespawnTime [45]
         maxSpawnedCount [-1]
         minRespawnTime [30]
         playlistTagQuery {autoDescription="", tokenStreamVersion=0, userDescription="", queryTokenStream [], tagDictionary []}
         tagDensityConditions []
        }
        pointProviderTagQueryOverride {autoDescription="", tokenStreamVersion=0, userDescription="", queryTokenStream [], tagDictionary []}
       }]
      encounterIdentifierTags {gameplayTags [{tagName=CAMP.AI.Goal.Assault.Wave.One}]}
      leashRadiusInner {registryType=None, value=200, curve {curveTable=null, rowName=None}}
      leashRadiusOuter {registryType=None, value=2500, curve {curveTable=null, rowName=None}}
      pointProviderTagQuery {
       autoDescription=" ANY( CAMP.Location.SpawnPoint.Wave1 )"
       tokenStreamVersion=0
       userDescription=""
       queryTokenStream [0,1,1,1,0]
       tagDictionary [{tagName=CAMP.Location.SpawnPoint.Wave1}]
      }
      spawnActions [
       {spawnActionClass=/Script/LagerRuntime.FortAthenaLivingWorldSpawnAction_AddAIToGroup, spawnActionTag {tagName=None}}
       {
        spawnActionClass=&/BP_PlayVFX_SpawnAction.BP_PlayVFX_SpawnAction_C
        spawnActionTag {tagName=None}
       }
      ]
      spawnTimeout {registryType=None, value=-1, curve {curveTable=null, rowName=None}}
     }
     node {
      name=Spawn Wave One
      _structType=/Script/EncountersRuntime.LWMEncounterDynamicStartStateTreeTask
      bCanEditConsideredForCompletion=true
      bConsideredForCompletion=true
      bHasTaskCompletionDelegateDispatcher=false
      bTaskEnabled=true
      transitionHandlingPriority=Normal
      bindingsBatch {value=65535}
      executionRuntimeTemplateIndex {value=65535}
      instanceDataHandle {index=65535, source=None, stateHandle {index=65535}}
      instanceTemplateIndex {value=65535}
      outputBindingsBatch {value=65535}
     }
    }
   }
   {
    id=#33
    description=Grant Quest
    properties {
     iD=#34
     instance {
      _structType=/Script/EncountersRuntime.GrantQuestStateTreeTaskInstanceData
      grantType=GrantToPlayspace
      questToGrant=/BattlePassS33_Quests/Items/QuestItems/Found/BatterBoi/Quest_S33_Scenario_BatterBoi_Playspace_Q01
      targetActor=null
      userActor=null
     }
     node {
      name=None
      _structType=/Script/EncountersRuntime.GrantQuestStateTreeTask
      bCanEditConsideredForCompletion=true
      bConsideredForCompletion=true
      bHasTaskCompletionDelegateDispatcher=false
      bTaskEnabled=true
      transitionHandlingPriority=Normal
      bindingsBatch {value=65535}
      executionRuntimeTemplateIndex {value=65535}
      instanceDataHandle {index=65535, source=None, stateHandle {index=65535}}
      instanceTemplateIndex {value=65535}
      outputBindingsBatch {value=65535}
     }
    }
   }
   {
    id=#35
    description=Set Target - BatterBoi
    properties {
     iD=#36
     instance {
      _structType=/Script/EncountersRuntime.LWMEncounterSetAlertLevelStateTreeTaskInstanceData
      targetActor=null
      targetBehavior=OneTime
      userActor=null
      encounterQuery {
       autoDescription=" ANY( CAMP.AI.Goal.Assault.Wave )"
       tokenStreamVersion=0
       userDescription=""
       queryTokenStream [0,1,1,1,0]
       tagDictionary [{tagName=CAMP.AI.Goal.Assault.Wave}]
      }
      targetEncounterQuery {autoDescription="", tokenStreamVersion=0, userDescription="", queryTokenStream [], tagDictionary []}
      targetInfo {awareForgetTime=15}
     }
     node {
      name=Set Target - BatterBoi
      _structType=/Script/EncountersRuntime.LWMEncounterSetAlertLevelStateTreeTask
      bCanEditConsideredForCompletion=true
      bConsideredForCompletion=true
      bHasTaskCompletionDelegateDispatcher=false
      bTaskEnabled=true
      transitionHandlingPriority=Normal
      bindingsBatch {value=65535}
      executionRuntimeTemplateIndex {value=65535}
      instanceDataHandle {index=65535, source=None, stateHandle {index=65535}}
      instanceTemplateIndex {value=65535}
      outputBindingsBatch {value=65535}
     }
    }
   }
   {
    id=#37
    description="LWM Encounter [Remaining Count Transition]"
    properties {
     iD=#38
     instance {
      _structType=/Script/EncountersRuntime.LWMEncounterRemainingCountTransitionStateTreeTaskInstanceData
      bOnlyTriggerOnce=true
      bShouldTransition=true
      bTriggered=false
      userActor=null
      value=0
      customVerb {amount=1, customVerbTag {tagName=None}, subjects [], values []}
      encounterQuery {
       autoDescription=" ANY( CAMP.AI.Goal.Assault.Wave.One )"
       tokenStreamVersion=0
       userDescription=""
       queryTokenStream [0,1,1,1,0]
       tagDictionary [{tagName=CAMP.AI.Goal.Assault.Wave.One}]
      }
      onTriggered {iD=#19}
     }
     node {
      name=None
      _structType=/Script/EncountersRuntime.LWMEncounterRemainingCountTransitionStateTreeTask
      bCanEditConsideredForCompletion=true
      bConsideredForCompletion=true
      bHasTaskCompletionDelegateDispatcher=false
      bTaskEnabled=true
      transitionHandlingPriority=Normal
      bindingsBatch {value=65535}
      executionRuntimeTemplateIndex {value=65535}
      instanceDataHandle {index=65535, source=None, stateHandle {index=65535}}
      instanceTemplateIndex {value=65535}
      outputBindingsBatch {value=65535}
      transitionTo {name=Stage_Wave_Two, fallback=None, iD=#39, linkType=GotoState, stateHandle {index=65535}}
     }
    }
   }
  ]
  transitions []
 }
 Stage_Wave_Two {
  id=#8
  type=STATE
  tasks [
   {
    id=#40
    description=Speech Bubble - BatterBoi
    properties {
     iD=#41
     instance {
      _structType=&GrantReactionStateTreeTaskInstanceData
      grantType=GrantToTargetActor
      targetActor=null
      userActor=null
      reactionEntries [{rowValue {
         bUseDataRegistryId=false
         dataRegistryId=""
         dataTableRow {dataTable=&DT_QuestScenario_BatterBoi_Reactions, rowName=SpeechBubble_WaveTwo}
        }}]
      reactions []
     }
     node {
      name=Speech Bubble - BatterBoi
      _structType=&GrantReactionStateTreeTask
      bCanEditConsideredForCompletion=true
      bConsideredForCompletion=true
      bHasTaskCompletionDelegateDispatcher=false
      bTaskEnabled=true
      transitionHandlingPriority=Normal
      bindingsBatch {value=65535}
      executionRuntimeTemplateIndex {value=65535}
      instanceDataHandle {index=65535, source=None, stateHandle {index=65535}}
      instanceTemplateIndex {value=65535}
      outputBindingsBatch {value=65535}
     }
    }
   }
   {
    id=#42
    description=Spawn Wave Two
    properties {
     iD=#43
     instance {
      _structType=&LWMEncounterDynamicStartStateTreeTaskInstanceData
      bAttachVolumeToSpawn=false
      bOverrideAnchorPosition=true
      bOverrideDelayBetweenEvents=false
      bOverrideEnvQuery=false
      bOverrideStageActorMaxCount=false
      bOverrideStageActorMaxSpawnedCount=false
      bPauseOnExit=false
      bSpecifyFaction=true
      bUseVolumeAsAnchor=true
      bWaitForEncounterDoneSpawning=false
      delayBetweenEventsMax=0
      delayBetweenEventsMin=0
      dynamicEmptyEncounter=&Lager_Encounter_Empty_QuestEncounter
      faction=&DA_QuestScenario_EnemyGroup
      instigator=null
      owningActor=null
      spawnAroundEnvironmentQuery=null
      spawnAroundEnvironmentQueryRunMode=AllMatching
      spawnAroundNavigationSourceOverride=""
      stageActorMaxCount=1
      stageActorMaxSpawnedCount=-1
      userActor=null
      actorTags {gameplayTags []}
      anchorPointTagQuery {autoDescription="", tokenStreamVersion=0, userDescription="", queryTokenStream [], tagDictionary []}
      dynamicEncounterEntries [{
        actorCountOverrideMax=1
        actorCountOverrideMin=1
        bCreateRandomSingleActorEvents=false
        bOverrideActorCount=false
        bUseEncounterTagQuery=true
        bUseEventTemplate=false
        debugName=""
        maxConcurrentSpawns=2
        noActorLimitEventCount=1
        spawnCount=2
        actorDescriptions []
        encounterTagQuery {
         autoDescription=" ANY( Athena.Quest.Scenario.Faction.Henchmen.Assault )"
         tokenStreamVersion=0
         userDescription=""
         queryTokenStream [0,1,1,1,0]
         tagDictionary [{tagName=Athena.Quest.Scenario.Faction.Henchmen.Assault}]
        }
        eventTemplate {
         activationGamePhaseStep=None
         activationMaxDelay=0
         activationMinDelay=0
         activationSafeZoneIndex=0
         bCanSpawnWithoutMatchingPlayerDistance=false
         bCheckForActorDensity=true
         bOnlyInNextSafeZone=false
         bOnlyInNextSafeZoneIfStormShrinking=false
         bOnlyInSafeZone=true
         bShouldProcessEventRequestImmediately=false
         deactivationMaxDelay=-1
         deactivationMinDelay=-1
         maxActorDensityOverride=-1
         maxDistanceToPlayer=-1
         minDistanceToPlayer=-1
         randomActivationChance=100
         weight=1
         calendarEventConditions {isActiveWithoutSeasonalManager=false, conditions []}
         canSpawnCondition {description=Empty, schemaClass=&FortAthenaLivingWorldConditionSchema, editableConditions [], sharedDefinition {}}
         despawnCondition {
          bCanBeDespawned=false
          bCanDespawnInTheStorm=false
          bCanDespawnWhenInCombatWithPlayer=false
          canStartDespawningPhaseStep=None
          canStartDespawningSafeZonePhase=-1
          maxStimulusAgeToBeConsideredInCombat=30
          minDistanceToPlayer=-1
          timeAsCandidateBeforeDespawn=10
         }
         despawnConditions []
         eventDatas []
         eventInstanceMaxCount [-1]
         eventInstanceMaxSpawnedCount [-1]
         instantDeathMaxRespawnTime []
         instantDeathMinRespawnTime []
         isEnabled {registryType=None, value=1, curve {curveTable=null, rowName=None}}
         maxCount [0]
         maxRespawnTime [45]
         maxSpawnedCount [-1]
         minRespawnTime [30]
         playlistTagQuery {autoDescription="", tokenStreamVersion=0, userDescription="", queryTokenStream [], tagDictionary []}
         tagDensityConditions []
        }
        pointProviderTagQueryOverride {autoDescription="", tokenStreamVersion=0, userDescription="", queryTokenStream [], tagDictionary []}
       }]
      encounterIdentifierTags {gameplayTags [{tagName=CAMP.AI.Goal.Assault.Wave.Two}]}
      leashRadiusInner {registryType=None, value=200, curve {curveTable=null, rowName=None}}
      leashRadiusOuter {registryType=None, value=0, curve {curveTable=null, rowName=None}}
      pointProviderTagQuery {
       autoDescription=" ANY( CAMP.Location.SpawnPoint.Wave2 )"
       tokenStreamVersion=0
       userDescription=""
       queryTokenStream [0,1,1,1,0]
       tagDictionary [{tagName=CAMP.Location.SpawnPoint.Wave2}]
      }
      spawnActions [
       {spawnActionClass=&FortAthenaLivingWorldSpawnAction_AddAIToGroup, spawnActionTag {tagName=None}}
       {
        spawnActionClass=&BP_PlayVFX_SpawnAction_C
        spawnActionTag {tagName=None}
       }
      ]
      spawnTimeout {registryType=None, value=-1, curve {curveTable=null, rowName=None}}
     }
     node {
      name=Spawn Wave Two
      _structType=&LWMEncounterDynamicStartStateTreeTask
      bCanEditConsideredForCompletion=true
      bConsideredForCompletion=true
      bHasTaskCompletionDelegateDispatcher=false
      bTaskEnabled=true
      transitionHandlingPriority=Normal
      bindingsBatch {value=65535}
      executionRuntimeTemplateIndex {value=65535}
      instanceDataHandle {index=65535, source=None, stateHandle {index=65535}}
      instanceTemplateIndex {value=65535}
      outputBindingsBatch {value=65535}
     }
    }
   }
   {
    id=#44
    description=Set Target - BatterBoi
    properties {
     iD=#45
     instance {
      _structType=&LWMEncounterSetAlertLevelStateTreeTaskInstanceData
      targetActor=null
      targetBehavior=OneTime
      userActor=null
      encounterQuery {
       autoDescription=" ANY( CAMP.AI.Goal.Assault.Wave )"
       tokenStreamVersion=0
       userDescription=""
       queryTokenStream [0,1,1,1,0]
       tagDictionary [{tagName=CAMP.AI.Goal.Assault.Wave}]
      }
      targetEncounterQuery {autoDescription="", tokenStreamVersion=0, userDescription="", queryTokenStream [], tagDictionary []}
      targetInfo {awareForgetTime=15}
     }
     node {
      name=Set Target - BatterBoi
      _structType=&LWMEncounterSetAlertLevelStateTreeTask
      bCanEditConsideredForCompletion=true
      bConsideredForCompletion=true
      bHasTaskCompletionDelegateDispatcher=false
      bTaskEnabled=true
      transitionHandlingPriority=Normal
      bindingsBatch {value=65535}
      executionRuntimeTemplateIndex {value=65535}
      instanceDataHandle {index=65535, source=None, stateHandle {index=65535}}
      instanceTemplateIndex {value=65535}
      outputBindingsBatch {value=65535}
     }
    }
   }
   {
    id=#46
    description="LWM Encounter [Remaining Count Transition]"
    properties {
     iD=#47
     instance {
      _structType=&LWMEncounterRemainingCountTransitionStateTreeTaskInstanceData
      bOnlyTriggerOnce=true
      bShouldTransition=true
      bTriggered=false
      userActor=null
      value=0
      customVerb {amount=1, customVerbTag {tagName=None}, subjects [], values []}
      encounterQuery {
       autoDescription=" ANY( CAMP.AI.Goal.Assault.Wave.Two )"
       tokenStreamVersion=0
       userDescription=""
       queryTokenStream [0,1,1,1,0]
       tagDictionary [{tagName=CAMP.AI.Goal.Assault.Wave.Two}]
      }
      onTriggered {iD=#19}
     }
     node {
      name=None
      _structType=&LWMEncounterRemainingCountTransitionStateTreeTask
      bCanEditConsideredForCompletion=true
      bConsideredForCompletion=true
      bHasTaskCompletionDelegateDispatcher=false
      bTaskEnabled=true
      transitionHandlingPriority=Normal
      bindingsBatch {value=65535}
      executionRuntimeTemplateIndex {value=65535}
      instanceDataHandle {index=65535, source=None, stateHandle {index=65535}}
      instanceTemplateIndex {value=65535}
      outputBindingsBatch {value=65535}
      transitionTo {name=Delay, fallback=None, iD=#48, linkType=GotoState, stateHandle {index=65535}}
     }
    }
   }
  ]
  transitions []
 }
 Delay {
  id=#9
  type=STATE
  tasks [
   {
    id=#49
    description=Spawn Actor - Smart Object
    properties {
     iD=#50
     instance {
      _structType=/Script/EncountersRuntime.SpawnActorStateTreeTaskInstanceData
      actorClass=/NPCLibraryCh6/SmartObjects/SmartObject_Flexing_NoMesh_Actor.SmartObject_Flexing_NoMesh_Actor_C
      bAttachEncounterVolume=false
      bEnablePhysics=false
      bUseActorLocation=false
      bWaitForSpawnedActor=false
      owningActor=null
      persistenceBehavior=Destroy
      spawnedActor=null
      targetActor=null
      userActor=null
      actorIdentifierTags {gameplayTags [{tagName=QuestEncounter.BatterBoi.RewardEmote}]}
      actorTags {gameplayTags [{tagName=QuestEncounter.BatterBoi.RewardEmote}]}
      anchorOffset=<0 0 0>
      anchorQuery {
       autoDescription=" ANY( QuestEncounter.BatterBoi.RewardLocation )"
       tokenStreamVersion=0
       userDescription=""
       queryTokenStream [0,1,1,1,0]
       tagDictionary [{tagName=QuestEncounter.BatterBoi.RewardLocation}]
      }
      existingActorQuery {autoDescription="", tokenStreamVersion=0, userDescription="", queryTokenStream [], tagDictionary []}
     }
     node {
      name=Spawn Actor - Smart Object
      _structType=/Script/EncountersRuntime.SpawnActorStateTreeTask
      bCanEditConsideredForCompletion=true
      bConsideredForCompletion=true
      bHasTaskCompletionDelegateDispatcher=false
      bTaskEnabled=true
      transitionHandlingPriority=Normal
      bindingsBatch {value=65535}
      executionRuntimeTemplateIndex {value=65535}
      instanceDataHandle {index=65535, source=None, stateHandle {index=65535}}
      instanceTemplateIndex {value=65535}
      outputBindingsBatch {value=65535}
     }
    }
   }
   {
    id=#51
    description=Delay 2.0
    properties {
     iD=#52
     instance {_structType=&StateTreeDelayTaskInstanceData, duration=2}
     node {name=None, _structType=&StateTreeDelayTask}
    }
   }
  ]
  transitions [{id=#53, trigger=ON_STATE_COMPLETED, state {id=#10, name=Stage_Wave_Three}}]
 }
 Stage_Wave_Three {
  id=#10
  type=STATE
  tasks [
   {
    id=#54
    description=Speech Bubble - BatterBoi
    properties {
     iD=#55
     instance {
      _structType=&GrantReactionStateTreeTaskInstanceData
      grantType=GrantToTargetActor
      targetActor=null
      userActor=null
      reactionEntries [{rowValue {
         bUseDataRegistryId=false
         dataRegistryId=""
         dataTableRow {
          dataTable=&DT_QuestScenario_BatterBoi_Reactions
          rowName=SpeechBubble_WaveThree
         }
        }}]
      reactions []
     }
     node {
      name=Speech Bubble - BatterBoi
      _structType=&GrantReactionStateTreeTask
      bCanEditConsideredForCompletion=true
      bConsideredForCompletion=true
      bHasTaskCompletionDelegateDispatcher=false
      bTaskEnabled=true
      transitionHandlingPriority=Normal
      bindingsBatch {value=65535}
      executionRuntimeTemplateIndex {value=65535}
      instanceDataHandle {index=65535, source=None, stateHandle {index=65535}}
      instanceTemplateIndex {value=65535}
      outputBindingsBatch {value=65535}
     }
    }
   }
   {
    id=#56
    description=Spawn Wave Three
    properties {
     iD=#57
     instance {
      _structType=&LWMEncounterDynamicStartStateTreeTaskInstanceData
      bAttachVolumeToSpawn=false
      bOverrideAnchorPosition=true
      bOverrideDelayBetweenEvents=false
      bOverrideEnvQuery=false
      bOverrideStageActorMaxCount=false
      bOverrideStageActorMaxSpawnedCount=false
      bPauseOnExit=false
      bSpecifyFaction=true
      bUseVolumeAsAnchor=true
      bWaitForEncounterDoneSpawning=false
      delayBetweenEventsMax=0
      delayBetweenEventsMin=0
      dynamicEmptyEncounter=&Lager_Encounter_Empty_QuestEncounter
      faction=&DA_QuestScenario_EnemyGroup
      instigator=null
      owningActor=null
      spawnAroundEnvironmentQuery=null
      spawnAroundEnvironmentQueryRunMode=AllMatching
      spawnAroundNavigationSourceOverride=""
      stageActorMaxCount=1
      stageActorMaxSpawnedCount=-1
      userActor=null
      actorTags {gameplayTags []}
      anchorPointTagQuery {autoDescription="", tokenStreamVersion=0, userDescription="", queryTokenStream [], tagDictionary []}
      dynamicEncounterEntries [{
        actorCountOverrideMax=1
        actorCountOverrideMin=1
        bCreateRandomSingleActorEvents=false
        bOverrideActorCount=false
        bUseEncounterTagQuery=true
        bUseEventTemplate=false
        debugName=""
        maxConcurrentSpawns=1
        noActorLimitEventCount=1
        spawnCount=1
        actorDescriptions []
        encounterTagQuery {
         autoDescription=" ANY( Athena.Quest.Scenario.Faction.Henchmen.Boss )"
         tokenStreamVersion=0
         userDescription=""
         queryTokenStream [0,1,1,1,0]
         tagDictionary [{tagName=Athena.Quest.Scenario.Faction.Henchmen.Boss}]
        }
        eventTemplate {
         activationGamePhaseStep=None
         activationMaxDelay=0
         activationMinDelay=0
         activationSafeZoneIndex=0
         bCanSpawnWithoutMatchingPlayerDistance=false
         bCheckForActorDensity=true
         bOnlyInNextSafeZone=false
         bOnlyInNextSafeZoneIfStormShrinking=false
         bOnlyInSafeZone=true
         bShouldProcessEventRequestImmediately=false
         deactivationMaxDelay=-1
         deactivationMinDelay=-1
         maxActorDensityOverride=-1
         maxDistanceToPlayer=-1
         minDistanceToPlayer=-1
         randomActivationChance=100
         weight=1
         calendarEventConditions {isActiveWithoutSeasonalManager=false, conditions []}
         canSpawnCondition {description=Empty, schemaClass=&FortAthenaLivingWorldConditionSchema, editableConditions [], sharedDefinition {}}
         despawnCondition {
          bCanBeDespawned=false
          bCanDespawnInTheStorm=false
          bCanDespawnWhenInCombatWithPlayer=false
          canStartDespawningPhaseStep=None
          canStartDespawningSafeZonePhase=-1
          maxStimulusAgeToBeConsideredInCombat=30
          minDistanceToPlayer=-1
          timeAsCandidateBeforeDespawn=10
         }
         despawnConditions []
         eventDatas []
         eventInstanceMaxCount [-1]
         eventInstanceMaxSpawnedCount [-1]
         instantDeathMaxRespawnTime []
         instantDeathMinRespawnTime []
         isEnabled {registryType=None, value=1, curve {curveTable=null, rowName=None}}
         maxCount [0]
         maxRespawnTime [45]
         maxSpawnedCount [-1]
         minRespawnTime [30]
         playlistTagQuery {autoDescription="", tokenStreamVersion=0, userDescription="", queryTokenStream [], tagDictionary []}
         tagDensityConditions []
        }
        pointProviderTagQueryOverride {autoDescription="", tokenStreamVersion=0, userDescription="", queryTokenStream [], tagDictionary []}
       }]
      encounterIdentifierTags {gameplayTags [{tagName=CAMP.AI.Goal.Assault.Wave.Three}]}
      leashRadiusInner {registryType=None, value=200, curve {curveTable=null, rowName=None}}
      leashRadiusOuter {registryType=None, value=2500, curve {curveTable=null, rowName=None}}
      pointProviderTagQuery {
       autoDescription=" ANY( CAMP.Location.SpawnPoint.Wave3 )"
       tokenStreamVersion=0
       userDescription=""
       queryTokenStream [0,1,1,1,0]
       tagDictionary [{tagName=CAMP.Location.SpawnPoint.Wave3}]
      }
      spawnActions [
       {spawnActionClass=&FortAthenaLivingWorldSpawnAction_AddAIToGroup, spawnActionTag {tagName=None}}
       {
        spawnActionClass=&BP_PlayVFX_SpawnAction_C
        spawnActionTag {tagName=None}
       }
      ]
      spawnTimeout {registryType=None, value=-1, curve {curveTable=null, rowName=None}}
     }
     node {
      name=Spawn Wave Three
      _structType=&LWMEncounterDynamicStartStateTreeTask
      bCanEditConsideredForCompletion=true
      bConsideredForCompletion=true
      bHasTaskCompletionDelegateDispatcher=false
      bTaskEnabled=true
      transitionHandlingPriority=Normal
      bindingsBatch {value=65535}
      executionRuntimeTemplateIndex {value=65535}
      instanceDataHandle {index=65535, source=None, stateHandle {index=65535}}
      instanceTemplateIndex {value=65535}
      outputBindingsBatch {value=65535}
     }
    }
   }
   {
    id=#58
    description=Set Target - BatterBoi
    properties {
     iD=#59
     instance {
      _structType=&LWMEncounterSetAlertLevelStateTreeTaskInstanceData
      targetActor=null
      targetBehavior=OneTime
      userActor=null
      encounterQuery {
       autoDescription=" ANY( CAMP.AI.Goal.Assault.Wave )"
       tokenStreamVersion=0
       userDescription=""
       queryTokenStream [0,1,1,1,0]
       tagDictionary [{tagName=CAMP.AI.Goal.Assault.Wave}]
      }
      targetEncounterQuery {autoDescription="", tokenStreamVersion=0, userDescription="", queryTokenStream [], tagDictionary []}
      targetInfo {awareForgetTime=15}
     }
     node {
      name=Set Target - BatterBoi
      _structType=&LWMEncounterSetAlertLevelStateTreeTask
      bCanEditConsideredForCompletion=true
      bConsideredForCompletion=true
      bHasTaskCompletionDelegateDispatcher=false
      bTaskEnabled=true
      transitionHandlingPriority=Normal
      bindingsBatch {value=65535}
      executionRuntimeTemplateIndex {value=65535}
      instanceDataHandle {index=65535, source=None, stateHandle {index=65535}}
      instanceTemplateIndex {value=65535}
      outputBindingsBatch {value=65535}
     }
    }
   }
   {
    id=#60
    description="LWM Encounter [Remaining Count Transition]"
    properties {
     iD=#61
     instance {
      _structType=&LWMEncounterRemainingCountTransitionStateTreeTaskInstanceData
      bOnlyTriggerOnce=true
      bShouldTransition=true
      bTriggered=false
      userActor=null
      value=0
      customVerb {amount=1, customVerbTag {tagName=None}, subjects [], values []}
      encounterQuery {
       autoDescription=" ANY( CAMP.AI.Goal.Assault.Wave.Three )"
       tokenStreamVersion=0
       userDescription=""
       queryTokenStream [0,1,1,1,0]
       tagDictionary [{tagName=CAMP.AI.Goal.Assault.Wave.Three}]
      }
      onTriggered {iD=#19}
     }
     node {
      name=None
      _structType=&LWMEncounterRemainingCountTransitionStateTreeTask
      bCanEditConsideredForCompletion=true
      bConsideredForCompletion=true
      bHasTaskCompletionDelegateDispatcher=false
      bTaskEnabled=true
      transitionHandlingPriority=Normal
      bindingsBatch {value=65535}
      executionRuntimeTemplateIndex {value=65535}
      instanceDataHandle {index=65535, source=None, stateHandle {index=65535}}
      instanceTemplateIndex {value=65535}
      outputBindingsBatch {value=65535}
      transitionTo {name=Speech_Bubble, fallback=None, iD=#62, linkType=GotoState, stateHandle {index=65535}}
     }
    }
   }
  ]
  transitions []
 }
 Go_To_Chest {
  id=#11
  type=STATE
  tasks [
   {
    id=#63
    description=Get Scenario Actor - Smart Object
    properties {
     iD=#64
     instance {
      _structType=&GetEncounterActorStateTreeTaskInstanceData
      actorQuery {autoDescription=" ANY( QuestEncounter.BatterBoi.RewardEmote )", tagDictionary [{tagName=QuestEncounter.BatterBoi.RewardEmote}]}
     }
     node {name=Get Scenario Actor - Smart Object, _structType=&GetEncounterActorStateTreeTask}
    }
   }
   {
    id=#65
    description=Enable Smart Objects
    properties {
     iD=#66
     instance {
      _structType=/Script/EncountersRuntime.EnableSmartObjectsStateTreeTaskInstanceData
      bEnable=true
      bUseTagQuery=false
      targetActor=null
      userActor=null
      aIActors []
      aICommandSOUsageData {bIgnoreCooldowns=false, requestDuration=10}
      actorQuery {autoDescription="", tokenStreamVersion=0, userDescription="", queryTokenStream [], tagDictionary []}
     }
     node {
      name=None
      _structType=/Script/EncountersRuntime.EnableSmartObjectsStateTreeTask
      bCanEditConsideredForCompletion=true
      bConsideredForCompletion=true
      bHasTaskCompletionDelegateDispatcher=false
      bTaskEnabled=true
      transitionHandlingPriority=Normal
      bindingsBatch {value=65535}
      executionRuntimeTemplateIndex {value=65535}
      instanceDataHandle {index=65535, source=None, stateHandle {index=65535}}
      instanceTemplateIndex {value=65535}
      outputBindingsBatch {value=65535}
     }
    }
   }
   {
    id=#67
    description=Delay 1.0
    properties {
     iD=#68
     instance {_structType=&StateTreeDelayTaskInstanceData, duration=1}
     node {name=None, _structType=&StateTreeDelayTask}
    }
   }
  ]
  transitions [{id=#69, trigger=ON_STATE_COMPLETED, state {id=#12, name=Speech_Bubble}}]
 }
 Speech_Bubble {
  id=#12
  type=STATE
  tasks [
   {
    id=#70
    description=Speech Bubble - BatterBoi
    properties {
     iD=#71
     instance {
      _structType=&GrantReactionStateTreeTaskInstanceData
      grantType=GrantToTargetActor
      targetActor=null
      userActor=null
      reactionEntries [{rowValue {
         bUseDataRegistryId=false
         dataRegistryId=""
         dataTableRow {
          dataTable=&DT_QuestScenario_BatterBoi_Reactions
          rowName=SpeechBubble_ScenarioEnd
         }
        }}]
      reactions []
     }
     node {
      name=Speech Bubble - BatterBoi
      _structType=&GrantReactionStateTreeTask
      bCanEditConsideredForCompletion=true
      bConsideredForCompletion=true
      bHasTaskCompletionDelegateDispatcher=false
      bTaskEnabled=true
      transitionHandlingPriority=Normal
      bindingsBatch {value=65535}
      executionRuntimeTemplateIndex {value=65535}
      instanceDataHandle {index=65535, source=None, stateHandle {index=65535}}
      instanceTemplateIndex {value=65535}
      outputBindingsBatch {value=65535}
     }
    }
   }
   {
    id=#72
    description=Delay 1.0
    properties {
     iD=#73
     instance {_structType=&StateTreeDelayTaskInstanceData, duration=1}
     node {name=None, _structType=&StateTreeDelayTask}
    }
   }
  ]
  transitions [{id=#74, trigger=ON_STATE_COMPLETED, state {id=#13, name=Spawn_Chest}}]
 }
 Spawn_Chest {
  id=#13
  type=STATE
  tasks [
   {
    id=#75
    description=Grant Quest
    properties {
     iD=#76
     instance {
      _structType=&GrantQuestStateTreeTaskInstanceData
      grantType=GrantToPlayspace
      questToGrant=/BattlePassS33_Quests/Items/QuestItems/Found/BatterBoi/Quest_S33_Scenario_BatterBoi_Playspace_Q02
      targetActor=null
      userActor=null
     }
     node {
      name=None
      _structType=&GrantQuestStateTreeTask
      bCanEditConsideredForCompletion=true
      bConsideredForCompletion=true
      bHasTaskCompletionDelegateDispatcher=false
      bTaskEnabled=true
      transitionHandlingPriority=Normal
      bindingsBatch {value=65535}
      executionRuntimeTemplateIndex {value=65535}
      instanceDataHandle {index=65535, source=None, stateHandle {index=65535}}
      instanceTemplateIndex {value=65535}
      outputBindingsBatch {value=65535}
     }
    }
   }
   {
    id=#77
    description=Spawn Chest
    properties {
     iD=#78
     instance {
      _structType=&SpawnActorStateTreeTaskInstanceData
      actorClass=&/AlwaysSpawn_RareChest_Scenario_BatterBoi.AlwaysSpawn_RareChest_Scenario_BatterBoi_C
      bAttachEncounterVolume=false
      bEnablePhysics=false
      bUseActorLocation=false
      bWaitForSpawnedActor=true
      owningActor=null
      persistenceBehavior=Destroy
      spawnedActor=null
      targetActor=null
      userActor=null
      actorIdentifierTags {gameplayTags [{tagName=QuestEncounter.BatterBoi.Prop}]}
      actorTags {gameplayTags [{tagName=QuestEncounter.BatterBoi.Prop}]}
      anchorOffset=<0 0 0>
      anchorQuery {
       autoDescription=" ANY( QuestEncounter.BatterBoi.Reward )"
       tokenStreamVersion=0
       userDescription=""
       queryTokenStream [0,1,1,1,0]
       tagDictionary [{tagName=QuestEncounter.BatterBoi.Reward}]
      }
      existingActorQuery {autoDescription="", tokenStreamVersion=0, userDescription="", queryTokenStream [], tagDictionary []}
     }
     node {
      name=Spawn Chest
      _structType=&SpawnActorStateTreeTask
      bCanEditConsideredForCompletion=true
      bConsideredForCompletion=true
      bHasTaskCompletionDelegateDispatcher=false
      bTaskEnabled=true
      transitionHandlingPriority=Normal
      bindingsBatch {value=65535}
      executionRuntimeTemplateIndex {value=65535}
      instanceDataHandle {index=65535, source=None, stateHandle {index=65535}}
      instanceTemplateIndex {value=65535}
      outputBindingsBatch {value=65535}
     }
    }
   }
   {
    id=#79
    description=Delay 1.0
    properties {
     iD=#80
     instance {_structType=&StateTreeDelayTaskInstanceData, duration=1}
     node {name=None, _structType=&StateTreeDelayTask}
    }
   }
  ]
  transitions [{id=#81, trigger=ON_STATE_COMPLETED, state {id=#14, name=Reward_Setup}}]
 }
 Reward_Setup {
  id=#14
  type=STATE
  tasks [
   {
    id=#82
    description=Get Scenario Actor - Chest
    properties {
     iD=#83
     instance {
      _structType=&GetEncounterActorStateTreeTaskInstanceData
      bQueryAgainstTrackedActorIdentifiers=true
      actorQuery {autoDescription=" ANY( QuestEncounter.BatterBoi.Prop )", tagDictionary [{tagName=QuestEncounter.BatterBoi.Prop}]}
     }
     node {name=Get Scenario Actor - Chest, _structType=&GetEncounterActorStateTreeTask}
    }
   }
   {
    id=#84
    description=Grant Reactions - Chest
    properties {
     iD=#85
     instance {
      _structType=&GrantReactionStateTreeTaskInstanceData
      grantType=GrantToTargetActor
      targetActor=null
      userActor=null
      reactionEntries [{rowValue {
         bUseDataRegistryId=false
         dataRegistryId=""
         dataTableRow {
          dataTable=&DT_QuestScenario_BatterBoi_Reactions
          rowName=IconVisibility_Visible
         }
        }}]
      reactions []
     }
     node {
      name=Grant Reactions - Chest
      _structType=&GrantReactionStateTreeTask
      bCanEditConsideredForCompletion=true
      bConsideredForCompletion=true
      bHasTaskCompletionDelegateDispatcher=false
      bTaskEnabled=true
      transitionHandlingPriority=Normal
      bindingsBatch {value=65535}
      executionRuntimeTemplateIndex {value=65535}
      instanceDataHandle {index=65535, source=None, stateHandle {index=65535}}
      instanceTemplateIndex {value=65535}
      outputBindingsBatch {value=65535}
     }
    }
   }
   {
    id=#86
    description=Scenario Verb Transition
    properties {
     iD=#87
     instance {
      _structType=/Script/EncountersRuntime.EncounterVerbTransitionStateTreeTaskInstanceData
      bIsTriggered=false
      bUseAdvancedObjectives=false
      objectiveValueThreshold=1
      userActor=null
      objectiveVerbs [{
        _structType=/Script/FortniteGame.FortObjectiveFilter_QuestUpdated
        bUseQuestDefinitionsFilter=false
        questState=Completed
        questSubtype=None
        questType=All
        author {
         bCanGainProgressForThisObjective=true
         tagQuery {autoDescription="", tokenStreamVersion=0, userDescription="", queryTokenStream [], tagDictionary []}
        }
        context {tagQuery {autoDescription="", tokenStreamVersion=0, userDescription="", queryTokenStream [], tagDictionary []}}
        questDefinitions {advancedQuestFilters []}
        questTags {tagQuery {
          autoDescription=" ANY( QuestEncounter.BatterBoi.Prop )"
          tokenStreamVersion=0
          userDescription=""
          queryTokenStream [0,1,1,1,0]
          tagDictionary [{tagName=QuestEncounter.BatterBoi.Prop}]
         }}
       }]
      objectives []
      onTriggered {iD=#19}
     }
     node {
      name=None
      _structType=/Script/EncountersRuntime.EncounterVerbTransitionStateTreeTask
      bCanEditConsideredForCompletion=true
      bConsideredForCompletion=true
      bHasTaskCompletionDelegateDispatcher=false
      bTaskEnabled=true
      transitionHandlingPriority=Normal
      bindingsBatch {value=65535}
      executionRuntimeTemplateIndex {value=65535}
      instanceDataHandle {index=65535, source=None, stateHandle {index=65535}}
      instanceTemplateIndex {value=65535}
      outputBindingsBatch {value=65535}
      transitionTo {name=Complete, fallback=None, iD=#88, linkType=GotoState, stateHandle {index=65535}}
     }
    }
   }
  ]
  transitions []
 }
 HideIcon_RemoveReactions {
  id=#15
  type=STATE
  tasks [
   {
    id=#89
    description=Get Scenario Actor - Chest
    properties {
     iD=#90
     instance {
      _structType=&GetEncounterActorStateTreeTaskInstanceData
      bQueryAgainstTrackedActorIdentifiers=true
      actorQuery {autoDescription=" ANY( QuestEncounter.BatterBoi.Prop )", tagDictionary [{tagName=QuestEncounter.BatterBoi.Prop}]}
     }
     node {name=Get Scenario Actor - Chest, _structType=&GetEncounterActorStateTreeTask}
    }
   }
   {
    id=#91
    description=Grant Reactions - Chest
    properties {
     iD=#92
     instance {
      _structType=&GrantReactionStateTreeTaskInstanceData
      grantType=GrantToTargetActor
      targetActor=null
      userActor=null
      reactionEntries [{rowValue {
         bUseDataRegistryId=false
         dataRegistryId=""
         dataTableRow {dataTable=&DT_QuestScenario_BatterBoi_Reactions, rowName=IconVisibility_Hide}
        }}]
      reactions []
     }
     node {
      name=Grant Reactions - Chest
      _structType=&GrantReactionStateTreeTask
      bCanEditConsideredForCompletion=true
      bConsideredForCompletion=true
      bHasTaskCompletionDelegateDispatcher=false
      bTaskEnabled=true
      transitionHandlingPriority=Normal
      bindingsBatch {value=65535}
      executionRuntimeTemplateIndex {value=65535}
      instanceDataHandle {index=65535, source=None, stateHandle {index=65535}}
      instanceTemplateIndex {value=65535}
      outputBindingsBatch {value=65535}
     }
    }
   }
   {
    id=#93
    description=Grant Reactions - BatterBoi
    properties {
     iD=#94
     instance {
      _structType=&GrantReactionStateTreeTaskInstanceData
      grantType=GrantToTargetActor
      targetActor=null
      userActor=null
      reactionEntries [
       {rowValue {
         bUseDataRegistryId=false
         dataRegistryId=""
         dataTableRow {dataTable=&DT_QuestScenario_BatterBoi_Reactions, rowName=IconVisibility_Hide}
        }}
       {rowValue {
         bUseDataRegistryId=false
         dataRegistryId=""
         dataTableRow {dataTable=&DT_QuestScenario_BatterBoi_Reactions, rowName=RemoveReactions}
        }}
      ]
      reactions []
     }
     node {
      name=Grant Reactions - BatterBoi
      _structType=&GrantReactionStateTreeTask
      bCanEditConsideredForCompletion=true
      bConsideredForCompletion=true
      bHasTaskCompletionDelegateDispatcher=false
      bTaskEnabled=true
      transitionHandlingPriority=Normal
      bindingsBatch {value=65535}
      executionRuntimeTemplateIndex {value=65535}
      instanceDataHandle {index=65535, source=None, stateHandle {index=65535}}
      instanceTemplateIndex {value=65535}
      outputBindingsBatch {value=65535}
     }
    }
   }
   {
    id=#95
    description=Get Scenario Actor - Smart Object
    properties {
     iD=#96
     instance {
      _structType=&GetEncounterActorStateTreeTaskInstanceData
      actorQuery {autoDescription=" ANY( QuestEncounter.BatterBoi.RewardEmote )", tagDictionary [{tagName=QuestEncounter.BatterBoi.RewardEmote}]}
     }
     node {name=Get Scenario Actor - Smart Object, _structType=&GetEncounterActorStateTreeTask}
    }
   }
   {
    id=#97
    description=Enable Smart Objects
    properties {
     iD=#98
     instance {
      _structType=&EnableSmartObjectsStateTreeTaskInstanceData
      bEnable=false
      bUseTagQuery=false
      targetActor=null
      userActor=null
      aIActors []
      aICommandSOUsageData {bIgnoreCooldowns=false, requestDuration=10}
      actorQuery {autoDescription="", tokenStreamVersion=0, userDescription="", queryTokenStream [], tagDictionary []}
     }
     node {
      name=None
      _structType=&EnableSmartObjectsStateTreeTask
      bCanEditConsideredForCompletion=true
      bConsideredForCompletion=true
      bHasTaskCompletionDelegateDispatcher=false
      bTaskEnabled=true
      transitionHandlingPriority=Normal
      bindingsBatch {value=65535}
      executionRuntimeTemplateIndex {value=65535}
      instanceDataHandle {index=65535, source=None, stateHandle {index=65535}}
      instanceTemplateIndex {value=65535}
      outputBindingsBatch {value=65535}
     }
    }
   }
   {
    id=#99
    description=Delay 1.0
    properties {
     iD=#100
     instance {_structType=&StateTreeDelayTaskInstanceData, duration=1}
     node {name=None, _structType=&StateTreeDelayTask}
    }
   }
  ]
  transitions [{id=#101, trigger=ON_STATE_COMPLETED, state {id=#16, name=Complete}}]
 }
 Complete {
  id=#16
  type=STATE
  tasks [
   {
    id=#102
    description=Send Custom Verb
    properties {
     iD=#103
     instance {
      _structType=/Script/EncountersRuntime.EncounterSendCustomVerbStateTreeTaskInstanceData
      bSendVerbForEachPlayerAuthor=false
      userActor=null
      customVerb {amount=1, customVerbTag {tagName=CAMP.Encounter.State.Complete}, subjects [], values []}
     }
     node {
      name=None
      _structType=/Script/EncountersRuntime.EncounterSendCustomVerbStateTreeTask
      bCanEditConsideredForCompletion=true
      bConsideredForCompletion=true
      bHasTaskCompletionDelegateDispatcher=false
      bTaskEnabled=true
      transitionHandlingPriority=Normal
      bindingsBatch {value=65535}
      executionRuntimeTemplateIndex {value=65535}
      instanceDataHandle {index=65535, source=None, stateHandle {index=65535}}
      instanceTemplateIndex {value=65535}
      outputBindingsBatch {value=65535}
     }
    }
   }
   {
    id=#104
    description=BP SDA STTask Complete State
    properties {
     iD=#105
     instanceObject=&self:StateTreeEditorData_0.StateTreeState_11.BP_SDA_STTask_CompleteState_C_0
     node {
      name=None
      _structType=&StateTreeBlueprintTaskWrapper
      taskClass=/StateDrivenActor/Tasks/BP_SDA_STTask_CompleteState.BP_SDA_STTask_CompleteState_C
     }
    }
   }
   {
    id=#106
    description="(Fortnite) Look Around Task"
    properties {
     iD=#107
     instance {
      _structType=/Script/FortniteAIServer.FortAthenaStateTreeLookAroundTaskInstanceData
      actor=null
      lookAtDuration=0
      lookAtDurationMax=5
      lookAtDurationMin=2
      timer=0
     }
     node {
      name=None
      _structType=/Script/FortniteAIServer.FortAthenaStateTreeLookAroundTask
      bCanEditConsideredForCompletion=true
      bConsideredForCompletion=true
      bHasTaskCompletionDelegateDispatcher=false
      bTaskEnabled=true
      transitionHandlingPriority=Normal
      bindingsBatch {value=65535}
      executionRuntimeTemplateIndex {value=65535}
      instanceDataHandle {index=65535, source=None, stateHandle {index=65535}}
      instanceTemplateIndex {value=65535}
      outputBindingsBatch {value=65535}
     }
    }
   }
  ]
  transitions [{id=#108, trigger=ON_STATE_COMPLETED, state {id=#109, name=Root, linkType=SUCCEEDED}}]
 }
]
